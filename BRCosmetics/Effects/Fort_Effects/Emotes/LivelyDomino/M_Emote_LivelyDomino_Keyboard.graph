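# Translucent, unlit, two-sided Niagara mesh and sprite material for the keyboard effect in the LivelyDomino emote.
# Emissive is a time-scrolled three-color blend loop tinted by particle color and scaled by intensity.
# Opacity combines particle alpha, a UV1 vertical falloff and a radial gradient swept by dynamic parameters; SweepX_position offsets both axes, leaving SweepY_position unused.
# Depth fade applies on default and deferred paths, while forward and mobile skip it.
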Material /BRCosmetics/Effects/Fort_Effects/Emotes/LivelyDomino/M_Emote_LivelyDomino_Keyboard
graph {
 properties {
  bAutomaticallySetUsageInEditor=false
  bUsedWithNiagaraMeshParticles=true
  bUsedWithNiagaraSprites=true
  blendMode=BLEND_Translucent
  shadingModel=MSM_Unlit
  twoSided=true
  usedShadingModels=Unlit
 }
 nodes [
  Constant2Vector() {g=-0.25, r=-0.5}
  Constant()
  n1 = Time()
  n2 = ScalarParameter() {defaultValue=-1, group=06_Emissive, parameterName=GradientScrollSpeed}
  n3 = MaterialFunctionCall!(A, B, C, Alpha=Frac(None=Multiply(A=@n1, B=@n2) {constB=-1})) {materialFunction=/Game/Global/MaterialFunctions/TechArt/MF_3ColorBlendLoop}
  n4 = ParticleColor()
  MaterialFunctionCall(Emissive=Multiply(A=ScalarParameter() {defaultValue=2, group=06_Emissive, parameterName=EmissiveIntensity, sliderMax=100}, B=Multiply(A=@n3.Result, B=@n4.RGB))) {materialFunction=&MF_FX_EmissiveAdjustments}
  n5 = DynamicParameter() {paramNames [sweep, SweepX_position, SweepY_position, Param4]}
  n6 = Add(A=TextureCoordinate() {vTiling=0.6}, B=AppendVector(A=@n5.SweepX_position, B=@n5.SweepX_position))
  n7 = Constant()
  n8 = DynamicParameter() {paramNames [sweep, SweepX_position, SweepY_position, Param4]}
  n9 = TextureSample(UVs=Add(
    A=MaterialFunctionCall(Vector or UVs=@n6) {materialFunction=&VectorToRadialValue}.Radial Coordinates
    B=AppendVector(A=@n7, B=@n8.sweep)
   )) {samplerSource=SSM_Clamp_WorldGroupSettings, texture=/Game/Global/Textures/Default/Masks_and_Gradients/gradient_stasis}
  n10 = TextureCoordinate() {coordinateIndex=1}
  n11 = Multiply(A=Multiply(A=@n4.A, B=Power(Base=@n9.R) {constExponent=1.5}), B=Power(Base=ComponentMask(None=@n10) {g=true}) {constExponent=1.5})
  n12 = DepthFade(Opacity=@n11, FadeDistance=ScalarParameter() {parameterName=DepthFade})
  ShadingPathSwitch(Default=@n12, Deferred=@n12, Forward=@n11, Mobile=@n11)
 ]
}
