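# Niagara effect for the LivelyDomino emote's keyboard, rendering mesh particles that, judging by the emitter name, play a sweep effect.
# A single emitter fires two bursts, the second slightly delayed in its own spawn group.
# Dynamic material parameters drive the look: a curve runs one value from positive to negative over particle life, while spawn group indexes arrays for two others, varying each pass.
# Particles get a small camera offset.
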NiagaraSystem /BRCosmetics/Effects/Fort_Effects/Emotes/LivelyDomino/NS_Emote_LivelyDomino_Keyboard
dependencies {
 usedDataInterfaces [&NiagaraDataInterfaceCurve, &NiagaraDataInterfaceArrayFloat]
 usedDynamicInputs []
 usedModules [&SystemState]
}
emitters {Sweep_one {
  emitterUpdateScript [
   &EmitterState(Inactive Response=::NewEnumerator1, Life Cycle Mode=::NewEnumerator1, Loop Behavior=::NewEnumerator1, Loop Duration=6, UseLoopDelay=false)
   &SpawnBurst_Instantaneous(Use Spawn Probability=false, UseLoopCountLimit=false)
   &SpawnBurst_Instantaneous(Spawn Group=1, Spawn Time=0.28, Use Spawn Probability=false, UseLoopCountLimit=false) {module=SpawnBurst_Instantaneous001}
  ]
  particleSpawnScript [
   &InitializeParticle(Color=<rgba 2 2 2 1>, Mesh Scale Mode=::NewEnumerator3, UsePositionOffset=true, Write Material Random=true)
   &CameraOffset(Camera Offset Amount=0.5)
  ]
  particleUpdateScript [
   &ParticleState()
   &DynamicMaterialParameters(
    Index 0 Param 1=&FloatFromCurve(FloatCurve=DataInterface() {propertyValues {Curve {keys [
         {arriveTangent=-3.147, interpMode=RCIM_Cubic, leaveTangent=-3.147, tangentMode=RCTM_User, time=0, value=1}
         {arriveTangent=-0.1771, interpMode=RCIM_Cubic, leaveTangent=-0.1771, tangentMode=RCTM_User, time=1, value=-1}
        ]}}})
    Index 0 Param 2=&SelectFloatFromArray(
     Direct Array Index=&SpawnGroup()
     Float Selection Array=DataInterface() {propertyValues {FloatData [-0.6,-0.75]}}
    )
    Index 0 Param 3=&SelectFloatFromArray(
     Direct Array Index=&SpawnGroup()
     Float Selection Array=DataInterface() {propertyValues {FloatData [-0.25,-0.3]}}
    )
    Index 0 Param 4=1
    Module.Index0Param0Write=true
    Module.Index0Param1Write=true
    Module.Index0Param2Write=true
    Module.Index0Param3Write=true
   )
  ]
  rendererClasses [&NiagaraMeshRendererProperties]
  renderers [{rendererClass=&NiagaraMeshRendererProperties, rendererIndex=0}]
 }}
summary {
 systemName=NS_Emote_LivelyDomino_Keyboard
 emitters [{emitterName=Sweep_one, rendererClasses [&NiagaraMeshRendererProperties]}]
 userVariables []
}
