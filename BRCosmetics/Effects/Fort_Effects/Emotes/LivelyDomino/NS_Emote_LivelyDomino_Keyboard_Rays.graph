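# Niagara effect for the LivelyDomino emote, named for keyboard rays, that fires a brief flash of light rays and sparks.
# Three burst emitters layer a mesh ray stretching along Z, a growing sprite glow, and a few small sprite sparks slowed by drag.
# Curves drive fades and dynamic material parameters; time of day scales brightness or alpha, and the ray also adjusts for platform and front end versus gameplay.
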NiagaraSystem /BRCosmetics/Effects/Fort_Effects/Emotes/LivelyDomino/NS_Emote_LivelyDomino_Keyboard_Rays
dependencies {
 usedDataInterfaces [&NiagaraDataInterfaceCurve, &NiagaraDataInterfacePlatformSet]
 usedDynamicInputs [&Multiply_Vector, &VectorFromFloat]
 usedModules [&SystemState]
}
emitters {
 Ray {
  emitterUpdateScript [
   &EmitterState(Inactive Response=::NewEnumerator1, Life Cycle Mode=::NewEnumerator1, Loop Behavior=::NewEnumerator1, Loop Duration=6, UseLoopDelay=false)
   &SpawnBurst_Instantaneous(Use Spawn Probability=false, UseLoopCountLimit=false)
  ]
  particleSpawnScript [
   &InitializeParticle(
    Lifetime=0.22
    Mesh Scale=&RandomRangeVector(Minimum=<1.5 1.5 3>, Maximum=<1.5 1.5 4>)
    Mesh Scale Mode=::NewEnumerator4
    Sprite Size=<15 10>
    Sprite Size Mode=::NewEnumerator4
    UsePositionOffset=true
    Write Material Random=true
   )
   &InitialMeshOrientation(Mesh Orientation Mode=::NewEnumerator2, Rotation=<0 -0.25 0>, Use Rotation Vector=true)
   &CameraOffset()
  ]
  particleUpdateScript [
   &ParticleState()
   &ScaleMeshSize(Scale Factor=&MakeVector(
     X=1
     Y=1
     Z=&FloatFromCurve(FloatCurve=DataInterface() {propertyValues {Curve {keys [{arriveTangent=2.2238, interpMode=RCIM_Cubic, leaveTangent=2.2238, tangentMode=RCTM_User, time=0, value=0.6}, <at 0.5 1 RCIM_Cubic>]}}})
    ))
   &ScaleColor(
    Scale Alpha=&FloatFromCurve(
     FloatCurve=DataInterface() {propertyValues {Curve {keys [
         <at 0 -0.05 RCIM_Cubic>
         {arriveTangent=0.0397, interpMode=RCIM_Cubic, leaveTangent=0.0397, tangentMode=RCTM_User, time=0.1, value=1}
         {arriveTangent=-1.6281, interpMode=RCIM_Cubic, leaveTangent=-1.6281, tangentMode=RCTM_User, time=0.6183, value=0.6305}
         {arriveTangent=-1.7372, interpMode=RCIM_Cubic, leaveTangent=-1.7372, tangentMode=RCTM_User, time=1, value=0}
        ]}}}
     Scale Curve=&Multiply_Float(
      A=&FNTimeOfDayScalar(Night Scale=0.5, Morning Scale=1, Evening Scale=1)
      B=&FN_PickFloatDependOnPlatform(TargetPlatformActive=1, TargetPlatformNoActive=0.5)
     )
    )
    Scale RGB=&Multiply_VectorByFloat(Vector=<1 1 1>, Float=&FN_FrontEndFloatModifier(Gameplay=2, Front End=1.5))
    ScaleA=true
    ScaleRGB=true
   )
   &DynamicMaterialParameters(
    Index 0 Param 3=&FloatFromCurve(FloatCurve=DataInterface() {propertyValues {Curve {keys [{arriveTangent=2.2562, interpMode=RCIM_Cubic, leaveTangent=2.2562, tangentMode=RCTM_User, time=0, value=0}, <at 1 1 RCIM_Cubic>]}}})
    Module.Index0Param0Write=true
    Module.Index0Param1Write=true
    Module.Index0Param2Write=true
    Module.Index0Param3Write=true
   )
  ]
  rendererClasses [&NiagaraMeshRendererProperties]
  renderers [{rendererClass=&NiagaraMeshRendererProperties, rendererIndex=0}]
 }
 glow {
  emitterUpdateScript [
   &EmitterState(Inactive Response=::NewEnumerator1, Life Cycle Mode=::NewEnumerator1, Loop Behavior=::NewEnumerator1, Loop Duration=6, UseLoopDelay=false)
   &SpawnBurst_Instantaneous(Use Spawn Probability=false, UseLoopCountLimit=false)
  ]
  particleSpawnScript [&InitializeParticle(
    Lifetime=0.3
    Position Offset=<-4.5 0 0>
    Sprite Size=&Multiply_Vector2D(A=<15 10>, B=&Vector2DFromFloat(Value=1.5))
    Sprite Size Mode=::NewEnumerator4
    UsePositionOffset=true
    Write Material Random=true
   )]
  particleUpdateScript [
   &ParticleState()
   &ScaleSpriteSize(Uniform Curve Sprite Scale=DataInterface() {propertyValues {Curve {keys [
        {arriveTangent=5.672, interpMode=RCIM_Cubic, leaveTangent=5.672, tangentMode=RCTM_User, time=0, value=0}
        {arriveTangent=0.9209, interpMode=RCIM_Cubic, leaveTangent=0.9209, tangentMode=RCTM_User, time=0.2, value=0.75}
        <at 1 1 RCIM_Cubic>
       ]}}})
   &ScaleColor(
    Scale Alpha=&FloatFromCurve(FloatCurve=DataInterface() {propertyValues {Curve {keys [
         <at 0 0 RCIM_Cubic>
         {arriveTangent=-0.0307, interpMode=RCIM_Cubic, leaveTangent=-0.0307, tangentMode=RCTM_User, time=0.05, value=1}
         {arriveTangent=-0.7852, interpMode=RCIM_Cubic, leaveTangent=-0.7852, tangentMode=RCTM_User, time=0.4, value=0.2}
         <at 1 0 RCIM_Cubic>
        ]}}})
    Scale RGB=&Multiply_VectorByFloat(
     Vector=<1 1 1>
     Float=&Multiply_Float(A=&FNTimeOfDayScalar(Night Scale=0.5, Morning Scale=1, Evening Scale=1), B=1)
    )
    ScaleA=true
    ScaleRGB=true
   )
   &DynamicMaterialParameters(
    Index 0 Param 1=0.5
    Index 0 Param 2=-0.5
    Index 0 Param 3=-0.3
    Module.Index0Param0Write=true
    Module.Index0Param1Write=true
    Module.Index0Param2Write=true
    Module.Index0Param3Write=true
   )
  ]
  rendererClasses [&NiagaraSpriteRendererProperties]
  renderers [{rendererClass=&NiagaraSpriteRendererProperties}]
 }
 particle {
  emitterUpdateScript [
   &EmitterState(Inactive Response=::NewEnumerator1, Life Cycle Mode=::NewEnumerator1, Loop Behavior=::NewEnumerator1, Loop Duration=6, UseLoopDelay=false)
   &SpawnBurst_Instantaneous(
    Spawn Count=&RandomRangeInt(Minimum=5, Maximum=10)
    Use Spawn Probability=false
    UseLoopCountLimit=false
   )
  ]
  particleSpawnScript [
   &InitializeParticle(
    Lifetime Max=0.4
    Lifetime Min=0.2
    Lifetime Mode=::NewEnumerator1
    Position Offset=<-4 0 0>
    Sprite Size=&Multiply_Vector2D(
     A=<10 6>
     B=&Vector2DFromFloat(Value=&RandomRangeFloat(Minimum=0.13, Maximum=0.25))
    )
    Sprite Size Mode=::NewEnumerator4
    UsePositionOffset=true
    Write Material Random=true
   )
   &AddVelocity(
    Module.Use Falloff Distance=false
    Origin Offset=<-4 0 0>
    Override Seed=false
    Velocity Mode=::NewEnumerator1
    Velocity Speed=&RandomRangeFloat(Minimum=30, Maximum=40)
   )
   &AddVelocity(Velocity=&RandomRangeVector(Minimum=<-15 0 0>, Maximum=<-25 0 0>)) {module=AddVelocity001}
  ]
  particleUpdateScript [
   &ParticleState()
   &ScaleSpriteSize(Uniform Curve Sprite Scale=DataInterface() {propertyValues {Curve {keys [
        {arriveTangent=5.672, interpMode=RCIM_Cubic, leaveTangent=5.672, tangentMode=RCTM_User, time=0, value=0}
        {arriveTangent=0.9209, interpMode=RCIM_Cubic, leaveTangent=0.9209, tangentMode=RCTM_User, time=0.2, value=0.75}
        <at 1 1 RCIM_Cubic>
       ]}}})
   &ScaleColor(
    Scale Alpha=&FloatFromCurve(FloatCurve=DataInterface() {propertyValues {Curve {keys [
         <at 0 0 RCIM_Cubic>
         {arriveTangent=-0.0307, interpMode=RCIM_Cubic, leaveTangent=-0.0307, tangentMode=RCTM_User, time=0.05, value=1}
         {arriveTangent=-0.7852, interpMode=RCIM_Cubic, leaveTangent=-0.7852, tangentMode=RCTM_User, time=0.4, value=0.2}
         <at 1 0 RCIM_Cubic>
        ]}}})
    Scale RGB=&Multiply_VectorByFloat(
     Vector=<30 30 30>
     Float=&Multiply_Float(A=&FNTimeOfDayScalar(Night Scale=0.5, Morning Scale=1, Evening Scale=1), B=1)
    )
    ScaleA=true
    ScaleRGB=true
   )
   &DynamicMaterialParameters(
    Index 0 Param 1=0.5
    Index 0 Param 2=-0.5
    Index 0 Param 3=-0.3
    Module.Index0Param0Write=true
    Module.Index0Param1Write=true
    Module.Index0Param2Write=true
    Module.Index0Param3Write=true
   )
   &Drag(Drag=2, Use Linear Drag=true, Use Rotational Drag=true)
   &SolveForcesAndVelocity(Clamp Velocity=false, Limit Acceleration=false)
  ]
  rendererClasses [&NiagaraSpriteRendererProperties]
  renderers [{rendererClass=&NiagaraSpriteRendererProperties}]
 }
}
summary {
 systemName=NS_Emote_LivelyDomino_Keyboard_Rays
 emitters [
  {emitterName=Ray, rendererClasses [&NiagaraMeshRendererProperties]}
  {emitterName=glow, rendererClasses [&NiagaraSpriteRendererProperties]}
  {emitterName=particle, rendererClasses [&NiagaraSpriteRendererProperties]}
 ]
 userVariables [
  User.bIsFrontEnd {type=NiagaraBool}
  User.bIsFrontEndPreview {type=NiagaraBool}
 ]
}
